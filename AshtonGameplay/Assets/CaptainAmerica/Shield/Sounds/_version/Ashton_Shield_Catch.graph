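# Sound cue for catching Captain America's shield, layering a random throw-catch sound with a random, slightly pitch-modulated grab sound.
# IsLocal and IsEnemy branches select among three copies of that layering: full volume for local, reduced for enemy, quietest for non-enemy others.
# An unset IsLocal plays the local mix, and an unset IsEnemy the enemy mix.
# It uses close weapon-projectile attenuation and a PreventNew concurrency override.
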SoundCue /AshtonGameplay/Assets/CaptainAmerica/Shield/Sounds/_version/Ashton_Shield_Catch
root=Branch_0
nodes [
 {id=Branch_0, children [Mixer_0,Branch_1,Mixer_0], properties {boolParameterName=IsLocal}}
 {id=Branch_1, children [Mixer_2,Mixer_4,Mixer_2], properties {boolParameterName=IsEnemy}}
 {id=Mixer_0, children [Random_1,Modulator_0], properties {inputVolume [1,1]}}
 {id=Mixer_1, children [Random_3,Modulator_1], properties {inputVolume [1,1]}}
 {id=Mixer_2, children [Mixer_1], properties {inputVolume [0.55]}}
 {id=Mixer_3, children [Random_7,Modulator_2], properties {inputVolume [1,1]}}
 {id=Mixer_4, children [Mixer_3], properties {inputVolume [0.25]}}
 {id=Modulator_0, children [Random_2], properties {pitchMax=1.1, pitchMin=1}}
 {id=Modulator_1, children [Random_6], properties {pitchMax=1.1, pitchMin=1}}
 {id=Modulator_2, children [Random_8], properties {pitchMax=1.1, pitchMin=1}}
 {id=Random_1, children [WavePlayer_2,WavePlayer_3,WavePlayer_4], properties {weights [1,1,1]}}
 {
  id=Random_2
  children [WavePlayer_20, WavePlayer_21, WavePlayer_22, WavePlayer_23]
  properties {bShouldExcludeFromBranchCulling=true, weights [1,1,1,1]}
 }
 {id=Random_3, children [WavePlayer_24, WavePlayer_25, WavePlayer_26], properties {weights [1,1,1]}}
 {
  id=Random_6
  children [WavePlayer_27, WavePlayer_28, WavePlayer_29, WavePlayer_30]
  properties {bShouldExcludeFromBranchCulling=true, weights [1,1,1,1]}
 }
 {id=Random_7, children [WavePlayer_31, WavePlayer_32, WavePlayer_33], properties {weights [1,1,1]}}
 {
  id=Random_8
  children [WavePlayer_34, WavePlayer_35, WavePlayer_36, WavePlayer_37]
  properties {bShouldExcludeFromBranchCulling=true, weights [1,1,1,1]}
 }
 {id=WavePlayer_2, properties {soundWaveAssetPtr=&/Ashton_Shield_Throw_Catch_01}}
 {id=WavePlayer_20, properties {soundWaveAssetPtr=/AshtonGameplay/Assets/CaptainAmerica/Shield/Sounds/Ashton_Shield_Catch_Grab_01}}
 {id=WavePlayer_21, properties {soundWaveAssetPtr=/AshtonGameplay/Assets/CaptainAmerica/Shield/Sounds/Ashton_Shield_Catch_Grab_02}}
 {id=WavePlayer_22, properties {soundWaveAssetPtr=/AshtonGameplay/Assets/CaptainAmerica/Shield/Sounds/Ashton_Shield_Catch_Grab_03}}
 {id=WavePlayer_23, properties {soundWaveAssetPtr=/AshtonGameplay/Assets/CaptainAmerica/Shield/Sounds/Ashton_Shield_Catch_Grab_04}}
 {id=WavePlayer_24, properties {soundWaveAssetPtr=&Ashton_Shield_Throw_Catch_01}}
 {id=WavePlayer_25, properties {soundWaveAssetPtr=&/Ashton_Shield_Throw_Catch_02}}
 {id=WavePlayer_26, properties {soundWaveAssetPtr=&/Ashton_Shield_Throw_Catch_03}}
 {id=WavePlayer_27, properties {soundWaveAssetPtr=&Ashton_Shield_Catch_Grab_01}}
 {id=WavePlayer_28, properties {soundWaveAssetPtr=&Ashton_Shield_Catch_Grab_02}}
 {id=WavePlayer_29, properties {soundWaveAssetPtr=&Ashton_Shield_Catch_Grab_03}}
 {id=WavePlayer_3, properties {soundWaveAssetPtr=&Ashton_Shield_Throw_Catch_02}}
 {id=WavePlayer_30, properties {soundWaveAssetPtr=&Ashton_Shield_Catch_Grab_04}}
 {id=WavePlayer_31, properties {soundWaveAssetPtr=&Ashton_Shield_Throw_Catch_01}}
 {id=WavePlayer_32, properties {soundWaveAssetPtr=&Ashton_Shield_Throw_Catch_02}}
 {id=WavePlayer_33, properties {soundWaveAssetPtr=&Ashton_Shield_Throw_Catch_03}}
 {id=WavePlayer_34, properties {soundWaveAssetPtr=&Ashton_Shield_Catch_Grab_01}}
 {id=WavePlayer_35, properties {soundWaveAssetPtr=&Ashton_Shield_Catch_Grab_02}}
 {id=WavePlayer_36, properties {soundWaveAssetPtr=&Ashton_Shield_Catch_Grab_03}}
 {id=WavePlayer_37, properties {soundWaveAssetPtr=&Ashton_Shield_Catch_Grab_04}}
 {id=WavePlayer_4, properties {soundWaveAssetPtr=&Ashton_Shield_Throw_Catch_03}}
]
properties {
 attenuationSettings=&Fort_Weapon_Projectile_Close_Athena_Att
 bOverrideConcurrency=true
 maxDistance=4950
 soundClassObject=&Fort_WeaponRanged_Fire_3P_PSM
 volumeMultiplier=1
 attenuationOverrides {falloffDistance=5000, attenuationShapeExtents {x=1000}}
 concurrencyOverrides {resolutionRule=PreventNew}
}
