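# Sound cue for equipping the Captain America shield, layering an equip sound with delayed foley.
# A mixer blends the main equip wave at full level with two quieter layers, each delayed briefly and picking randomly from throw foley variations.
# Modulators vary pitch per layer, pitching the first foley layer down.
# It uses weapon foley class, reload attenuation and a concurrency override limiting simultaneous instances.
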SoundCue /AshtonGameplay/Assets/CaptainAmerica/Shield/Sounds/_version/Ashton_Shield_Equip_Cue
root=Mixer_0
nodes [
 {id=Delay_0, children [Random_4], properties {delayMax=0.2, delayMin=0.2}}
 {id=Delay_2, children [Random_2], properties {delayMax=0.25, delayMin=0.25}}
 {id=Mixer_0, children [Modulator_0,Modulator_1,Modulator_3], properties {inputVolume [1,0.4,0.15]}}
 {id=Modulator_0, children [WavePlayer_0], properties {pitchMin=1}}
 {id=Modulator_1, children [Delay_0], properties {pitchMax=0.95, pitchMin=0.85}}
 {id=Modulator_3, children [Delay_2]}
 {id=Random_2, children [WavePlayer_5,WavePlayer_6,WavePlayer_8], properties {weights [1,1,1]}}
 {id=Random_4, children [WavePlayer_2,WavePlayer_4], properties {weights [1,1]}}
 {id=WavePlayer_0, properties {soundWaveAssetPtr=&/Ashton_Shield_Equip}}
 {id=WavePlayer_2, properties {soundWaveAssetPtr=&/Ashton_Shield_Throw_FoleyA_01}}
 {id=WavePlayer_4, properties {soundWaveAssetPtr=&/Ashton_Shield_Throw_FoleyA_03}}
 {id=WavePlayer_5, properties {soundWaveAssetPtr=&/Ashton_Shield_Throw_FoleyB_01}}
 {id=WavePlayer_6, properties {soundWaveAssetPtr=&/Ashton_Shield_Throw_FoleyB_02}}
 {id=WavePlayer_8, properties {soundWaveAssetPtr=&/Ashton_Shield_Throw_FoleyB_04}}
]
properties {
 attenuationSettings=&Fort_Weapon_Reload_Att
 bDebug=true
 bOverrideConcurrency=true
 maxDistance=2250
 soundClassObject=&Fort_WeaponFoley
 volumeMultiplier=0.45
 concurrencyOverrides {maxCount=3}
}
