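# Sound cue for the primary fire of Captain America's shield, layering a synth throw whoosh with a throw release sound.
# A mixer blends a randomly chosen whoosh at reduced volume, passed through a Doppler node, with a randomly chosen release sample pitched down by a modulator.
# It uses a third-person ranged-weapon fire sound class, in-air projectile throw attenuation and a concurrency override that limits simultaneous instances.
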SoundCue /AshtonGameplay/Assets/CaptainAmerica/Shield/Sounds/_version/Ashton_Shield_Primary_Fire
root=Mixer_0
nodes [
 {id=Doppler_0, children [Random_0], properties {dopplerIntensity=0.5}}
 {id=Mixer_0, children [Doppler_0,Modulator_1], properties {inputVolume [0.5,1]}}
 {id=Modulator_1, children [Random_3], properties {pitchMax=0.8, pitchMin=0.75}}
 {id=Random_0, children [WavePlayer_0,WavePlayer_1,WavePlayer_3], properties {bShouldExcludeFromBranchCulling=true, weights [1,1,1]}}
 {
  id=Random_3
  children [WavePlayer_14, WavePlayer_15, WavePlayer_16, WavePlayer_17]
  properties {bShouldExcludeFromBranchCulling=true, weights [1,1,1,1]}
 }
 {
  id=WavePlayer_0
  properties {soundWaveAssetPtr=&/Ashton_Shield_Throw_Whoosh_synth_01}
 }
 {
  id=WavePlayer_1
  properties {soundWaveAssetPtr=&/Ashton_Shield_Throw_Whoosh_synth_02}
 }
 {id=WavePlayer_14, properties {soundWaveAssetPtr=&/Ashton_Shield_Throw_Release_01}}
 {id=WavePlayer_15, properties {soundWaveAssetPtr=&/Ashton_Shield_Throw_Release_02}}
 {id=WavePlayer_16, properties {soundWaveAssetPtr=&/Ashton_Shield_Throw_Release_03}}
 {id=WavePlayer_17, properties {soundWaveAssetPtr=&/Ashton_Shield_Throw_Release_04}}
 {
  id=WavePlayer_3
  properties {soundWaveAssetPtr=&/Ashton_Shield_Throw_Whoosh_synth_04}
 }
]
properties {
 attenuationSettings=/Game/Sounds/Attenuation/Fort_Weapon_Projectile_InAir_Launch_Throw_Att
 bOverrideConcurrency=true
 maxDistance=8000
 soundClassObject=&Fort_WeaponRanged_Fire_3P_PSM
 volumeMultiplier=2
 attenuationOverrides {falloffDistance=5000, attenuationShapeExtents {x=1000}}
 concurrencyOverrides {maxCount=3}
}
