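# Sound cue for the foley of Captain America's shield primary-fire throw, with attenuation and a concurrency override.
# An IsLocal branch plays a four-layer local mix (foley A, delayed foley C, delayed energy whoosh, throw start), also used when the parameter is unset.
# Otherwise, an IsEnemy branch plays only foley A and throw start, louder for enemies or unset than for non-enemies.
# Each layer picks a random wave variation.
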SoundCue /AshtonGameplay/Assets/CaptainAmerica/Shield/Sounds/_version/Ashton_Shield_Primary_Fire_Foley
root=Branch_0
nodes [
 {id=Branch_0, children [Mixer_0,Branch_1,Mixer_0], properties {boolParameterName=IsLocal}}
 {id=Branch_1, children [Mixer_2,Mixer_4,Mixer_2], properties {boolParameterName=IsEnemy}}
 {id=Delay_0, children [Random_3], properties {delayMax=0.1, delayMin=0.1}}
 {id=Delay_1, children [Random_0], properties {delayMax=0.312, delayMin=0.3}}
 {id=Mixer_0, children [Random_4,Delay_1,Delay_0,Random_1], properties {inputVolume [0.6,0.6,0.2,1.2]}}
 {id=Mixer_1, children [Random_8,Random_5], properties {inputVolume [0.6,1.2]}}
 {id=Mixer_2, children [Mixer_1], properties {inputVolume [0.75]}}
 {id=Mixer_3, children [Random_11,Random_10], properties {inputVolume [0.6,1.2]}}
 {id=Mixer_4, children [Mixer_3], properties {inputVolume [0.25]}}
 {id=Random_0, children [WavePlayer_9, WavePlayer_18, WavePlayer_19], properties {bShouldExcludeFromBranchCulling=true, weights [1,1,1]}}
 {
  id=Random_1
  children [WavePlayer_0, WavePlayer_1, WavePlayer_10, WavePlayer_11, WavePlayer_12, WavePlayer_13]
  properties {bShouldExcludeFromBranchCulling=true, weights [1,1,1,1,1,1]}
 }
 {
  id=Random_10
  children [WavePlayer_36, WavePlayer_37, WavePlayer_38, WavePlayer_39, WavePlayer_40, WavePlayer_41]
  properties {bShouldExcludeFromBranchCulling=true, weights [1,1,1,1,1,1]}
 }
 {id=Random_11, children [WavePlayer_42, WavePlayer_43, WavePlayer_44], properties {bShouldExcludeFromBranchCulling=true, weights [1,1,1]}}
 {id=Random_3, children [WavePlayer_14, WavePlayer_16], properties {bShouldExcludeFromBranchCulling=true, weights [1,1]}}
 {id=Random_4, children [WavePlayer_2,WavePlayer_3,WavePlayer_4], properties {bShouldExcludeFromBranchCulling=true, weights [1,1,1]}}
 {
  id=Random_5
  children [WavePlayer_15, WavePlayer_17, WavePlayer_20, WavePlayer_21, WavePlayer_22, WavePlayer_23]
  properties {bShouldExcludeFromBranchCulling=true, weights [1,1,1,1,1,1]}
 }
 {id=Random_8, children [WavePlayer_24, WavePlayer_25, WavePlayer_26], properties {bShouldExcludeFromBranchCulling=true, weights [1,1,1]}}
 {id=WavePlayer_0, properties {soundWaveAssetPtr=&/Ashton_Shield_Throw_Start_01}}
 {id=WavePlayer_1, properties {soundWaveAssetPtr=&/Ashton_Shield_Throw_Start_02}}
 {id=WavePlayer_10, properties {soundWaveAssetPtr=&/Ashton_Shield_Throw_Start_03}}
 {id=WavePlayer_11, properties {soundWaveAssetPtr=&/Ashton_Shield_Throw_Start_04}}
 {id=WavePlayer_12, properties {soundWaveAssetPtr=&/Ashton_Shield_Throw_Start_05}}
 {id=WavePlayer_13, properties {soundWaveAssetPtr=&/Ashton_Shield_Throw_Start_06}}
 {
  id=WavePlayer_14
  properties {soundWaveAssetPtr=&/Ashton_Shield_Throw_Whoosh_energy_03}
 }
 {id=WavePlayer_15, properties {soundWaveAssetPtr=&Ashton_Shield_Throw_Start_01}}
 {
  id=WavePlayer_16
  properties {soundWaveAssetPtr=&/Ashton_Shield_Throw_Whoosh_energy_01}
 }
 {id=WavePlayer_17, properties {soundWaveAssetPtr=&Ashton_Shield_Throw_Start_02}}
 {id=WavePlayer_18, properties {soundWaveAssetPtr=&/Ashton_Shield_Throw_FoleyC_02}}
 {id=WavePlayer_19, properties {soundWaveAssetPtr=&/Ashton_Shield_Throw_FoleyC_03}}
 {id=WavePlayer_2, properties {soundWaveAssetPtr=&/Ashton_Shield_Throw_FoleyA_01}}
 {id=WavePlayer_20, properties {soundWaveAssetPtr=&Ashton_Shield_Throw_Start_03}}
 {id=WavePlayer_21, properties {soundWaveAssetPtr=&Ashton_Shield_Throw_Start_04}}
 {id=WavePlayer_22, properties {soundWaveAssetPtr=&Ashton_Shield_Throw_Start_05}}
 {id=WavePlayer_23, properties {soundWaveAssetPtr=&Ashton_Shield_Throw_Start_06}}
 {id=WavePlayer_24, properties {soundWaveAssetPtr=&Ashton_Shield_Throw_FoleyA_01}}
 {id=WavePlayer_25, properties {soundWaveAssetPtr=&/Ashton_Shield_Throw_FoleyA_02}}
 {id=WavePlayer_26, properties {soundWaveAssetPtr=&/Ashton_Shield_Throw_FoleyA_03}}
 {id=WavePlayer_3, properties {soundWaveAssetPtr=&Ashton_Shield_Throw_FoleyA_02}}
 {id=WavePlayer_36, properties {soundWaveAssetPtr=&Ashton_Shield_Throw_Start_01}}
 {id=WavePlayer_37, properties {soundWaveAssetPtr=&Ashton_Shield_Throw_Start_02}}
 {id=WavePlayer_38, properties {soundWaveAssetPtr=&Ashton_Shield_Throw_Start_03}}
 {id=WavePlayer_39, properties {soundWaveAssetPtr=&Ashton_Shield_Throw_Start_04}}
 {id=WavePlayer_4, properties {soundWaveAssetPtr=&Ashton_Shield_Throw_FoleyA_03}}
 {id=WavePlayer_40, properties {soundWaveAssetPtr=&Ashton_Shield_Throw_Start_05}}
 {id=WavePlayer_41, properties {soundWaveAssetPtr=&Ashton_Shield_Throw_Start_06}}
 {id=WavePlayer_42, properties {soundWaveAssetPtr=&Ashton_Shield_Throw_FoleyA_01}}
 {id=WavePlayer_43, properties {soundWaveAssetPtr=&Ashton_Shield_Throw_FoleyA_02}}
 {id=WavePlayer_44, properties {soundWaveAssetPtr=&Ashton_Shield_Throw_FoleyA_03}}
 {id=WavePlayer_9, properties {soundWaveAssetPtr=&/Ashton_Shield_Throw_FoleyC_01}}
]
properties {
 attenuationSettings=/Game/Sounds/Attenuation/Fort_Weapon_Projectile_InAir_Launch_Throw_Att
 bOverrideConcurrency=true
 maxDistance=8000
 soundClassObject=&Fort_WeaponRanged_Fire_3P_PSM
 volumeMultiplier=1.1
 attenuationOverrides {falloffDistance=5000, attenuationShapeExtents {x=1000}}
 concurrencyOverrides {maxCount=3}
}
