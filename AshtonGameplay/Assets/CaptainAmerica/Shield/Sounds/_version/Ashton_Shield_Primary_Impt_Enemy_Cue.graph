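# Sound cue for Captain America's thrown shield striking an enemy, named as the primary impact.
# An ImpactCount integer switch selects between standard impact sounds, also used as the default input, and bounce impact sounds.
# Each set picks an equally weighted random wave and passes through a volume-boosting mixer.
# Attenuation and concurrency are overridden on the cue, with natural-sound falloff and a small voice limit that stops the lowest priority.
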SoundCue /AshtonGameplay/Assets/CaptainAmerica/Shield/Sounds/_version/Ashton_Shield_Primary_Impt_Enemy_Cue
root=Switch_0
nodes [
 {id=Mixer_0, children [Random_0], properties {inputVolume [1.5]}}
 {id=Mixer_1, children [Random_1], properties {inputVolume [1.5]}}
 {
  id=Random_0
  children [WavePlayer_12, WavePlayer_13, WavePlayer_14, WavePlayer_16, WavePlayer_17]
  properties {bShouldExcludeFromBranchCulling=true, weights [1,1,1,1,1]}
 }
 {
  id=Random_1
  children [WavePlayer_0, WavePlayer_1, WavePlayer_8, WavePlayer_9, WavePlayer_10, WavePlayer_11]
  properties {bShouldExcludeFromBranchCulling=true, weights [1,1,1,1,1,1]}
 }
 {id=Switch_0, children [Mixer_1,Mixer_1,Mixer_0], properties {intParameterName=ImpactCount}}
 {
  id=WavePlayer_0
  properties {soundWaveAssetPtr=&/Ashton_Shield_Throw_Enemy_Impact_01}
 }
 {
  id=WavePlayer_1
  properties {soundWaveAssetPtr=&/Ashton_Shield_Throw_Enemy_Impact_02}
 }
 {
  id=WavePlayer_10
  properties {soundWaveAssetPtr=&/Ashton_Shield_Throw_Enemy_Impact_05}
 }
 {
  id=WavePlayer_11
  properties {soundWaveAssetPtr=&/Ashton_Shield_Throw_Enemy_Impact_06}
 }
 {
  id=WavePlayer_12
  properties {soundWaveAssetPtr=&/Ashton_Shield_Throw_Enemy_Impact_Bounce_01}
 }
 {
  id=WavePlayer_13
  properties {soundWaveAssetPtr=&/Ashton_Shield_Throw_Enemy_Impact_Bounce_02}
 }
 {
  id=WavePlayer_14
  properties {soundWaveAssetPtr=&/Ashton_Shield_Throw_Enemy_Impact_Bounce_03}
 }
 {
  id=WavePlayer_16
  properties {soundWaveAssetPtr=&/Ashton_Shield_Throw_Enemy_Impact_Bounce_05}
 }
 {
  id=WavePlayer_17
  properties {soundWaveAssetPtr=&/Ashton_Shield_Throw_Enemy_Impact_Bounce_06}
 }
 {
  id=WavePlayer_8
  properties {soundWaveAssetPtr=&/Ashton_Shield_Throw_Enemy_Impact_03}
 }
 {
  id=WavePlayer_9
  properties {soundWaveAssetPtr=&/Ashton_Shield_Throw_Enemy_Impact_04}
 }
]
properties {
 bOverrideAttenuation=true
 bOverrideConcurrency=true
 maxDistance=7000
 soundClassObject=/Game/Sounds/FortSoundClassesAndModes/Fort_Enemy_Damaged_PSM
 volumeMultiplier=1.2
 attenuationOverrides {distanceAlgorithm=NaturalSound, falloffDistance=6000, attenuationShapeExtents {x=1000}}
 concurrencyOverrides {maxCount=3, resolutionRule=StopLowestPriority}
}
