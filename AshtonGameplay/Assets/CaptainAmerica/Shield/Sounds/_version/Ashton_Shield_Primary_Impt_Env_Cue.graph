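# Sound cue for environment impacts of Captain America's thrown shield, in the Fort_Impacts class.
# An ImpactCount switch plays main impacts when unset or zero and bounce variants at one.
# Each branch mixes a random hit with a shared random debris layer, pitched up through a modulator and delay node.
# Overrides add HRTF spatialization, long natural falloff with filtering and occlusion, limited concurrency and a distance-based reverb send.
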SoundCue /AshtonGameplay/Assets/CaptainAmerica/Shield/Sounds/_version/Ashton_Shield_Primary_Impt_Env_Cue
root=Switch_0
nodes [
 {id=Delay_0, children [Modulator_0]}
 {id=Mixer_0, children [Random_0,Delay_0], properties {inputVolume [2,2.2]}}
 {id=Mixer_1, children [Random_1,Delay_0], properties {inputVolume [2,2.2]}}
 {id=Modulator_0, children [Random_2], properties {pitchMax=1.25, pitchMin=1.15}}
 {
  id=Random_0
  children [WavePlayer_8, WavePlayer_9, WavePlayer_14, WavePlayer_15, WavePlayer_16, WavePlayer_17]
  properties {weights [1,1,1,1,1,1]}
 }
 {id=Random_1, children [WavePlayer_2,WavePlayer_3,WavePlayer_4,WavePlayer_5,WavePlayer_6,WavePlayer_7], properties {weights [1,1,1,1,1,1]}}
 {id=Random_2, children [WavePlayer_0, WavePlayer_1, WavePlayer_10, WavePlayer_11], properties {weights [1,1,1,1]}}
 {id=Switch_0, children [Mixer_1,Mixer_1,Mixer_0], properties {intParameterName=ImpactCount}}
 {
  id=WavePlayer_0
  properties {soundWaveAssetPtr=&/Ashton_Shield_Throw_Env_Impact_Debris_01}
 }
 {
  id=WavePlayer_1
  properties {soundWaveAssetPtr=&/Ashton_Shield_Throw_Env_Impact_Debris_02}
 }
 {
  id=WavePlayer_10
  properties {soundWaveAssetPtr=&/Ashton_Shield_Throw_Env_Impact_Debris_03}
 }
 {
  id=WavePlayer_11
  properties {soundWaveAssetPtr=&/Ashton_Shield_Throw_Env_Impact_Debris_04}
 }
 {
  id=WavePlayer_14
  properties {soundWaveAssetPtr=&/Ashton_Shield_Throw_Env_Impact_Bounce_03}
 }
 {
  id=WavePlayer_15
  properties {soundWaveAssetPtr=&/Ashton_Shield_Throw_Env_Impact_Bounce_04}
 }
 {
  id=WavePlayer_16
  properties {soundWaveAssetPtr=&/Ashton_Shield_Throw_Env_Impact_Bounce_05}
 }
 {
  id=WavePlayer_17
  properties {soundWaveAssetPtr=&/Ashton_Shield_Throw_Env_Impact_Bounce_06}
 }
 {
  id=WavePlayer_2
  properties {soundWaveAssetPtr=&/Ashton_Shield_Throw_Env_Impact_01}
 }
 {
  id=WavePlayer_3
  properties {soundWaveAssetPtr=&/Ashton_Shield_Throw_Env_Impact_02}
 }
 {
  id=WavePlayer_4
  properties {soundWaveAssetPtr=&/Ashton_Shield_Throw_Env_Impact_03}
 }
 {
  id=WavePlayer_5
  properties {soundWaveAssetPtr=&/Ashton_Shield_Throw_Env_Impact_04}
 }
 {
  id=WavePlayer_6
  properties {soundWaveAssetPtr=&/Ashton_Shield_Throw_Env_Impact_05}
 }
 {
  id=WavePlayer_7
  properties {soundWaveAssetPtr=&/Ashton_Shield_Throw_Env_Impact_06}
 }
 {
  id=WavePlayer_8
  properties {soundWaveAssetPtr=&/Ashton_Shield_Throw_Env_Impact_Bounce_01}
 }
 {
  id=WavePlayer_9
  properties {soundWaveAssetPtr=&/Ashton_Shield_Throw_Env_Impact_Bounce_02}
 }
]
properties {
 bOverrideAttenuation=true
 bOverrideConcurrency=true
 maxDistance=17000
 soundClassObject=&Fort_Impacts
 volumeMultiplier=1.5
 attenuationOverrides {
  bAttenuateWithLPF=true
  bEnableLogFrequencyScaling=true
  distanceAlgorithm=NaturalSound
  falloffDistance=16000
  hPFFrequencyAtMax=200
  lPFFrequencyAtMax=2000
  lPFRadiusMax=3000
  lPFRadiusMin=1000
  occlusionInterpolationTime=0.5
  occlusionTraceChannel=ECC_GameTraceChannel6
  occlusionVolumeAttenuation=0.6
  spatializationAlgorithm=SPATIALIZATION_HRTF
  stereoSpread=0
  attenuationShapeExtents {x=1000}
 }
 concurrencyOverrides {maxCount=3}
 soundSubmixSends [{
   maxSendDistance=4000
   maxSendLevel=0.3
   minSendDistance=1000
   minSendLevel=0.01
   sendLevelControlMethod=Linear
   soundSubmix=&Fortnite_Reverb_Distant
  }]
}
