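# Looping sound cue for Captain America's thrown shield in the air, classed as a ranged weapon sound.
# A Mixer blends four looping throw waves at different volumes, two of them passed through random pitch Modulators for variation.
# An Enveloper quickly fades in volume and shapes pitch over time, then feeds a smoothed Doppler node at the root.
# An in-air attenuation asset handles spatialization, with concurrency overridden to limit simultaneous instances.
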SoundCue /AshtonGameplay/Assets/CaptainAmerica/Shield/Sounds/_version/Ashton_Shield_Primary_Loop_Cue
root=Doppler_0
nodes [
 {id=Doppler_0, children [Enveloper_1], properties {bUseSmoothing=true, dopplerIntensity=0.75}}
 {
  id=Enveloper_1
  children [Mixer_0]
  properties {
   pitchMax=0.8
   pitchMin=0.8
   pitchCurve {editorCurveData {keys [
      {time=0}
      {arriveTangent=0.0217, interpMode=RCIM_Cubic, leaveTangent=-0.1508, tangentMode=RCTM_Break, time=0.15, value=1.5}
      {arriveTangent=-0.0165, interpMode=RCIM_Cubic, leaveTangent=-0.0165, tangentMode=RCTM_User, time=3, value=1.3}
     ]}}
   volumeCurve {editorCurveData {keys [<at 0 0 RCIM_Cubic>, {arriveTangent=-0.0839, interpMode=RCIM_Cubic, leaveTangent=-0.0839, tangentMode=RCTM_User, time=0.25, value=1}]}}
  }
 }
 {id=Mixer_0, children [Modulator_0,WavePlayer_1,WavePlayer_3,Modulator_6], properties {inputVolume [0.4,1.2,1.2,0.8]}}
 {id=Modulator_0, children [WavePlayer_0], properties {pitchMax=1.25, pitchMin=1.15}}
 {id=Modulator_6, children [WavePlayer_2], properties {pitchMax=1, pitchMin=0.9}}
 {
  id=WavePlayer_0
  properties {bLooping=true, soundWaveAssetPtr=&/Ashton_Shield_ThrowLoopA_01}
 }
 {
  id=WavePlayer_1
  properties {bLooping=true, soundWaveAssetPtr=&/Ashton_Shield_ThrowLoopB_01}
 }
 {
  id=WavePlayer_2
  properties {bLooping=true, soundWaveAssetPtr=&/Ashton_Shield_Throw_Loop}
 }
 {
  id=WavePlayer_3
  properties {bLooping=true, soundWaveAssetPtr=&/Ashton_Shield_ThrowLoopC_01}
 }
]
properties {
 attenuationSettings=/AshtonGameplay/Assets/CaptainAmerica/Shield/Sounds/Att_Ashton_Shield_InAirLoop
 bOverrideConcurrency=true
 maxDistance=22500
 soundClassObject=&Fort_WeaponRanged
 volumeMultiplier=2
 attenuationOverrides {falloffDistance=5000, attenuationShapeExtents {x=1000}}
 concurrencyOverrides {maxCount=3}
}
