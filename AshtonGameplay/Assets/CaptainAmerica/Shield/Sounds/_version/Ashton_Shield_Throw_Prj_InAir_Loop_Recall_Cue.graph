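# Sound cue for Captain America's thrown shield during its in-air recall, layering return whooshes over looping flight sound.
# A mixer blends a return whoosh, a random pick of four pitch-varied charge whooshes, and two modulated throw loops.
# An enveloper quickly fades in the loops while their pitch spikes and then settles lower.
# A smoothed Doppler node at the root shifts the whole mix, with dedicated attenuation and a concurrency override capping simultaneous instances.
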SoundCue /AshtonGameplay/Assets/CaptainAmerica/Shield/Sounds/_version/Ashton_Shield_Throw_Prj_InAir_Loop_Recall_Cue
root=Doppler_0
nodes [
 {id=Doppler_0, children [Mixer_0], properties {bUseSmoothing=true, dopplerIntensity=0.35}}
 {
  id=Enveloper_0
  children [Mixer_1]
  properties {
   pitchMax=0.8
   pitchMin=0.8
   pitchCurve {editorCurveData {keys [{time=0}, {arriveTangent=0.0217, interpMode=RCIM_Cubic, leaveTangent=-0.7447, tangentMode=RCTM_Break, time=0.15, value=1.5}, <at 2 1.2>]}}
   volumeCurve {editorCurveData {keys [<at 0 0 RCIM_Cubic>, {arriveTangent=-0.0839, interpMode=RCIM_Cubic, leaveTangent=-0.0839, tangentMode=RCTM_User, time=0.15, value=1}]}}
  }
 }
 {id=Mixer_0, children [Modulator_5,Modulator_1,Enveloper_0], properties {inputVolume [0.8,0.7,1]}}
 {id=Mixer_1, children [Modulator_3,Modulator_6], properties {inputVolume [2,2]}}
 {id=Modulator_1, children [Random_0], properties {pitchMax=1.5, pitchMin=1}}
 {id=Modulator_3, children [WavePlayer_3], properties {pitchMax=1.1, pitchMin=1}}
 {id=Modulator_5, children [WavePlayer_4]}
 {id=Modulator_6, children [WavePlayer_2], properties {pitchMax=1, volumeMax=1, volumeMin=1}}
 {
  id=Random_0
  children [WavePlayer_22, WavePlayer_23, WavePlayer_24, WavePlayer_25]
  properties {bShouldExcludeFromBranchCulling=true, weights [1,1,1,1]}
 }
 {
  id=WavePlayer_2
  properties {bLooping=true, soundWaveAssetPtr=&/Ashton_Shield_Throw_Loop}
 }
 {id=WavePlayer_22, properties {soundWaveAssetPtr=/AshtonGameplay/Assets/CaptainAmerica/Shield/Sounds/Ashton_Shield_Charge_Whoosh_04}}
 {id=WavePlayer_23, properties {soundWaveAssetPtr=/AshtonGameplay/Assets/CaptainAmerica/Shield/Sounds/Ashton_Shield_Charge_Whoosh_03}}
 {id=WavePlayer_24, properties {soundWaveAssetPtr=/AshtonGameplay/Assets/CaptainAmerica/Shield/Sounds/Ashton_Shield_Charge_Whoosh_02}}
 {id=WavePlayer_25, properties {soundWaveAssetPtr=/AshtonGameplay/Assets/CaptainAmerica/Shield/Sounds/Ashton_Shield_Charge_Whoosh_01}}
 {
  id=WavePlayer_3
  properties {bLooping=true, soundWaveAssetPtr=&/Ashton_Shield_ThrowLoopC_01}
 }
 {id=WavePlayer_4, properties {soundWaveAssetPtr=&/Ashton_Shield_Return_Whoosh_01}}
]
properties {
 attenuationSettings=/AshtonGameplay/Assets/CaptainAmerica/Shield/Sounds/Att_Ashton_Shield_InAirLoop
 bOverrideConcurrency=true
 maxDistance=22500
 soundClassObject=&Fort_WeaponRanged_Fire_3P_PSM
 volumeMultiplier=1.1
 concurrencyOverrides {maxCount=3}
}
